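# Conversation database for the Durrburger Worker (female) NPC, defining what this character says when a player starts talking to it.
# The entry point leads to one greeting speech node, whose default line can be replaced by prioritized contextual lines for first-ever conversations, a per-NPC conversation count check, or specific player outfits.
# The speech node then branches through link nodes into the shared NPC Library quests and basic services conversations.
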ConversationDatabase /NPCLibrary/NPCs/DurrburgerWorker_F/NPCLibrary_Conversation_DurrburgerWorker_F
entry_points [{tag=Conversation.Entry.DurrburgerWorker_F, destinations [#1]}]
nodes [
 ConversationEntryPointNode_7 {
  id=#1
  class=ConversationEntryPointNode
  outputs [#2]
  properties {entryTag {tagName=Conversation.Entry.DurrburgerWorker_F}}
 }
 Speech_C_4 {
  id=#2
  class=Speech_C
  outputs [#3, #4]
  properties {
   nodeName=""
   generalConfig {
    defaultMessage=Order up?
    contextualMessages [
     {
      message=You here to grill me?
      contextRequirements [{requirement {class=FortControllerRequirement_IsFirstEverConversationWithNPC}}]
      priority {curve {rowName=Tandem.Speech.Priority.FirstEverInteraction}}
     }
     {
      message=Order up?
      contextRequirements [{requirement {
         class=FortControllerRequirement_CheckNPCMatchConversationCount
         properties {nPCUniqueIdTag {tagName=AISpawnerData.Type.Specialist.Medic.DurrburgerWorker_F}}
        }}]
      priority {curve {rowName=None}}
     }
     {
      message="If you're interested in changing places, the answer is still no."
      contextRequirements [{requirement {
         class=FortControllerRequirement_HasCID
         properties {allowedCIds [&Character_UndergroundRebel_Fashion]}
        }}]
      priority {value=2, curve {rowName=Tandem.Speech.Priority.OutfitReaction}}
     }
     {
      message=Why does everyone keep expecting me to fix reality?
      contextRequirements [{requirement {
         class=FortControllerRequirement_HasCID
         properties {allowedCIds [
           &Character_IronLilac_Sly
           &Character_BrightDisk
           &Character_VoidRedemption_Rebel
          ]}
        }}]
      priority {value=2, curve {rowName=Tandem.Speech.Priority.OutfitReaction}}
     }
    ]
   }
  }
 }
 ConversationLinkNode_0 {id=#4, class=ConversationLinkNode, properties {remoteEntryTag {tagName=Conversation.Entry.BasicServices}}}
 ConversationLinkNode_1 {id=#3, class=ConversationLinkNode, properties {remoteEntryTag {tagName=Conversation.Entry.NPCLibrary.Quests}}}
]
